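# Placeholder-style lit material for a broken-mesh FX asset in the SlapSplash consumable, apparently generated on FBX import.
# It is flagged for use with Niagara mesh particles.
# The graph holds only a light grey vector parameter, with no textures, math or visible connections to material outputs.
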Material /ChronoConsumables/Assets/SlapSplash/FX/_obj_BrokenMesh_fbxskinimport1
graph {
 properties {bUsedWithNiagaraMeshParticles=true, pixelDepthOffsetMode=PDOM_Legacy, usedShadingModels=Default Lit}
 nodes [VectorParameter() {defaultValue=<rgba 0.7569 0.7569 0.7569 0>}]
}
