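# Physics asset for the Slap Splash consumable mesh, giving its splash bones simple collision shapes for physics.
# It has a tiny capsule body on Root and a larger, individually sized and oriented capsule on each of the four dyn_splash bones.
# Each splash bone connects straight to Root, not in a chain, through a constraint with tight, equal swing and twist limit angles, suggesting slight secondary jiggle.
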PhysicsAsset /ChronoConsumables/Assets/SlapSplash/Meshes/Slap_Splash_Consumable_PhysicsAsset
bodies {
 Root [{shapeType=Capsule, length=1, radius=1}]
 dyn_splash_01 [{shapeType=Capsule, length=11.0514, radius=4.5724, center=<10.628 0.3296 -0.0185>, rotation=<rot 0 -89.4016 -89.8647>}]
 dyn_splash_02 [{shapeType=Capsule, length=9.0514, radius=6.5518, center=<10.8356 1.2887 0.2634>, rotation=<rot 90 0 0>}]
 dyn_splash_03 [{shapeType=Capsule, length=20.6027, radius=5.6423, center=<9.8291 0.3437 0.1799>, rotation=<rot 0 91.1201 -89.2736>}]
 dyn_splash_04 [{shapeType=Capsule, length=11.1027, radius=3.3923, center=<8 -0.1368 0.6536>, rotation=<rot 90 0 0>}]
}
constraints [
 {bone1Name=dyn_splash_01, bone2Name=Root, swing1LimitDegrees=5, swing2LimitDegrees=5, twistLimitDegrees=5}
 {bone1Name=dyn_splash_02, bone2Name=Root, swing1LimitDegrees=5, swing2LimitDegrees=5, twistLimitDegrees=5}
 {bone1Name=dyn_splash_03, bone2Name=Root, swing1LimitDegrees=5, swing2LimitDegrees=5, twistLimitDegrees=5}
 {bone1Name=dyn_splash_04, bone2Name=Root, swing1LimitDegrees=5, swing2LimitDegrees=5, twistLimitDegrees=5}
]
